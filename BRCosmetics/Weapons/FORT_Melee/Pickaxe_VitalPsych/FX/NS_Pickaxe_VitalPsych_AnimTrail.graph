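# Niagara ribbon trail system for the VitalPsych pickaxe, drawing three anim trails between bone or socket pairs.
# HardEdge001 varies width and alpha per trail, fading from overbright blue through color-over-life curves.
# SoftEdge adds pale, shorter-tapered ribbons with faint fade-in/out and speed-limited inherited velocity.
# Both taper width over life, loop on User.LoopDuration and use User.ColorSwitcher to choose or blend between two tints.
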NiagaraSystem /BRCosmetics/Weapons/FORT_Melee/Pickaxe_VitalPsych/FX/NS_Pickaxe_VitalPsych_AnimTrail
dependencies {
 usedDataInterfaces [&NiagaraDataInterfaceCurve, &NiagaraDataInterfaceArrayFloat, &NiagaraDataInterfaceColorCurve]
 usedDynamicInputs []
 usedModules [&SystemState]
}
emitters {
 HardEdge001 {
  emitterUpdateScript [
   &EmitterState(Life Cycle Mode=::NewEnumerator1, Loop Behavior=::NewEnumerator1, Loop Duration=$User.LoopDuration, UseLoopDelay=false)
   &FN_SpawnAnimTrail_v2(AnimTrailWidthMode=::FromSecond, Number of Anim Trails=3)
  ]
  particleSpawnScript [
   &FN_SetSpawnGroup(Number of Spawn Groups=3, Set Ribbon ID and Ribbon Link Order=true) {module=FN_SetSpawnGroup001}
   /Game/Effects/Niagara/Modules/AnimTrail/FN_InitializeAnimTrail/FN_InitializeAnimTrailBoneIndecies(Second Bone or Socket Index=1)
   &FN_InitializeAnimTrailBoneIndecies(AnimTrail Selection=1, First Bone or Socket Index=2, Second Bone or Socket Index=3) {module=FN_InitializeAnimTrailBoneIndecies001}
   &FN_InitializeAnimTrailBoneIndecies(AnimTrail Selection=2, First Bone or Socket Index=4, Second Bone or Socket Index=5) {module=FN_InitializeAnimTrailBoneIndecies002}
   /Game/Effects/Niagara/Modules/AnimTrail/FN_InitializeAnimTrail/FN_InitializeAnimTrail(
    AnimTrailWidth:NiagaraFloat=&SelectFloatFromArray(Direct Array Index=$Particles.SpawnGroup, Float Selection Array=DataInterface() {propertyValues {FloatData [1,1.7,1]}})
    Lifetime:NiagaraFloat=$User.Lifetime
    Position Offset:Vector3f=<0 0 2>
   )
  ]
  particleUpdateScript [
   &ScaleRibbonWidth(Ribbon Width Scale=&FloatFromCurve(FloatCurve=DataInterface() {propertyValues {Curve {keys [{arriveTangent=0.1159, interpMode=RCIM_Cubic, time=0, value=1}, {interpMode=RCIM_Cubic, leaveTangent=-0.9664, time=1, value=0}]}}}))
   &ScaleColor(
    Scale Alpha=&FloatFromCurve(FloatCurve=DataInterface() {propertyValues {Curve {keys [<at 0 0.25>,<at 1 0>]}}}, Scale Curve=10)
    Scale RGB=&VectorFromFloat(Value=&FloatFromCurve(FloatCurve=DataInterface() {propertyValues {Curve {keys [<at 0 1 RCIM_Cubic>, {arriveTangent=-1, interpMode=RCIM_Cubic, leaveTangent=-1, time=0.2, value=2}, <at 1 0 RCIM_Cubic>]}}}))
    ScaleA=true
    ScaleRGB=true
   ) {enabled=false}
   &ScaleColor(
    Scale Alpha=&FloatFromCurve(FloatCurve=DataInterface() {propertyValues {Curve {keys [<at 0 0.25>,<at 1 0>]}}}, Scale Curve=0.25)
    ScaleA=true
    ScaleRGB=false
   ) {enabled=false, module=ScaleColor001}
   &FN_UpdateAnimTrail_v2()
   &SolveForcesAndVelocity(Clamp Velocity=false, Limit Acceleration=false)
   &Color(
    Color=&LerpLinearColors(
     StartColor=&ColorFromCurve(ColorCurve=DataInterface() {propertyValues {
        AlphaCurve {keys [<at 0.1301 1 RCIM_Cubic>,<at 1 0 RCIM_Cubic>]}
        BlueCurve {keys [<at 0 10 RCIM_Cubic>,<at 0.3334 0.8714>]}
        GreenCurve {keys [<at 0 5.7319 RCIM_Cubic>,<at 0.3334 0.4213>]}
        RedCurve {keys [<at 0 3.8021 RCIM_Cubic>,<at 0.3334 0.2178>]}
       }})
     EndColor=&ColorFromCurve(ColorCurve=DataInterface() {propertyValues {
        AlphaCurve {keys [<at 0.1301 1 RCIM_Cubic>,<at 1 0 RCIM_Cubic>]}
        BlueCurve {keys [<at 0 8 RCIM_Cubic>,<at 0.3334 0.8125>]}
        GreenCurve {keys [<at 0 3.9507 RCIM_Cubic>,<at 0.3334 0.1727>]}
        RedCurve {keys [<at 0 3.0417 RCIM_Cubic>,<at 0.3334 0.0423>]}
       }})
     LerpFactor=&SelectFloatFromArray(
      Direct Array Index=&MakeIntFromFloat(Float=$User.ColorSwitcher)
      Float Selection Array=DataInterface() {propertyValues {FloatData [0,1]}}
     )
    )
    Scale Alpha=&SelectFloatFromArray(Direct Array Index=$Particles.SpawnGroup, Float Selection Array=DataInterface() {propertyValues {FloatData [0.05,0.1,0.05]}})
    Scale Color=&VectorFromFloat(Value=1.2)
   )
  ]
  rendererClasses [&NiagaraRibbonRendererProperties]
  renderers [{rendererClass=&NiagaraRibbonRendererProperties}]
 }
 SoftEdge {
  emitterUpdateScript [
   &EmitterState(Life Cycle Mode=::NewEnumerator1, Loop Behavior=::NewEnumerator1, Loop Duration=$User.LoopDuration, UseLoopDelay=false)
   &FN_SpawnAnimTrail_v2(AnimTrailWidthMode=::FromSecond, Number of Anim Trails=3)
  ]
  particleSpawnScript [
   &FN_InitializeAnimTrailBoneIndecies(Second Bone or Socket Index=1)
   &FN_InitializeAnimTrailBoneIndecies(AnimTrail Selection=1, First Bone or Socket Index=2, Second Bone or Socket Index=3) {module=FN_InitializeAnimTrailBoneIndecies001}
   &FN_InitializeAnimTrailBoneIndecies(AnimTrail Selection=2, First Bone or Socket Index=4, Second Bone or Socket Index=5) {module=FN_InitializeAnimTrailBoneIndecies002}
   &FN_InitializeAnimTrail(
    AnimTrailWidth=0.8
    Color=&LerpLinearColors(StartColor=<rgba 0.849 0.9169 1 0.6>, EndColor=<rgba 0.849 0.8741 1 0.6>, LerpFactor=$User.ColorSwitcher)
    Lifetime=0.145
    Position Offset:Vector3f=<0 0 2>
   )
  ]
  particleUpdateScript [
   &ScaleRibbonWidth(Ribbon Width Scale=&FloatFromCurve(FloatCurve=DataInterface() {propertyValues {Curve {keys [
         {arriveTangent=0.1159, interpMode=RCIM_Cubic, time=0, value=1}
         {arriveTangent=-1.2462, interpMode=RCIM_Cubic, leaveTangent=-1.2462, tangentMode=RCTM_User, time=1, value=0}
        ]}}}))
   &ScaleColor(
    Scale Alpha=&FloatFromCurve(FloatCurve=DataInterface() {propertyValues {Curve {keys [<at 0 0.25>,<at 1 0>]}}}, Scale Curve=10)
    Scale RGB=&VectorFromFloat(Value=&FloatFromCurve(FloatCurve=DataInterface() {propertyValues {Curve {keys [<at 0 1 RCIM_Cubic>, {arriveTangent=-1, interpMode=RCIM_Cubic, leaveTangent=-1, time=0.2, value=2}, <at 1 0 RCIM_Cubic>]}}}))
    ScaleA=true
    ScaleRGB=true
   ) {enabled=false}
   &ScaleColor(
    Scale Alpha=&FloatFromCurve(FloatCurve=DataInterface() {propertyValues {Curve {keys [<at 0 0 RCIM_Cubic>,<at 0.5 1 RCIM_Cubic>,<at 1 0 RCIM_Cubic>]}}}, Scale Curve=0.15)
    ScaleA=true
    ScaleRGB=true
   ) {module=ScaleColor001}
   &FN_UpdateAnimTrail_v2()
   &InheritVelocity(
    Inherited Velocity Amount Scale=&VectorFromFloat(Value=0.5)
    Inherited Velocity Speed Limit=50
    Speed Limit=true
    UseSpeedThreshold=false
   )
   &SolveForcesAndVelocity(Clamp Velocity=false, Limit Acceleration=false)
  ]
  rendererClasses [&NiagaraRibbonRendererProperties]
  renderers [{rendererClass=&NiagaraRibbonRendererProperties}]
 }
}
summary {
 systemName=NS_Pickaxe_VitalPsych_AnimTrail
 emitters [
  {emitterName=SoftEdge, rendererClasses [&NiagaraRibbonRendererProperties]}
  {emitterName=HardEdge001, rendererClasses [&NiagaraRibbonRendererProperties]}
 ]
 userVariables [
  User.AnimTrailWidth {type=NiagaraFloat, defaultValue=1, description=""}
  User.ColorSwitcher {type=NiagaraFloat, defaultValue=0, description=""}
  User.Lifetime {type=NiagaraFloat, defaultValue=0.13, description=""}
  User.LoopDuration {type=NiagaraFloat, defaultValue=0.2, description=""}
  User.Ribbon camera face {type=Vector3f, description="", defaultValue {struct=Vector3f, value=<0 0 0>}}
  User.RibbonIDOffset {type=NiagaraInt32}
  User.SkeletalMesh {
   type=NiagaraDataInterfaceSkeletalMesh
   defaultValue {
    struct=NiagaraExt_VariableValue_DataInterface
    value {dataInterface=&self:NiagaraDataInterfaceSkeletalMesh_0}
   }
  }
 ]
}
